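# Physics asset for the Tactical Santa Mask hat cosmetic, giving simulation bodies to its dangling hat, beard and mustache bones.
# Limited-swing, low-twist constraints chain three hat bones and two two-bone mustache sides to the head, and three beard bones to the jaw.
# Capsules and spheres on the head, jaw, upper spine and right hand supply body shapes, linked by constraints with free swing and twist.
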PhysicsAsset /BRCosmetics/Accessories/Hats/M_MED_Tactica_Santa_Mask/Meshes/M_MED_Tactical_Santa_Mask_Dyn
bodies {
 C_jaw [{shapeType=Capsule, length=2, radius=2, rotation=<rot 90 0 0>}]
 dyn_beard_01 [{shapeType=Capsule, length=7.3999, radius=5, center=<1.8732 0 0>, rotation=<rot 0 0 90>}]
 dyn_beard_02 [{shapeType=Capsule, length=6.1693, radius=5, center=<1.8732 0 0>, rotation=<rot 0 0 90>}]
 dyn_beard_03 [{shapeType=Capsule, length=5, radius=5, center=<1.8732 0 0>, rotation=<rot 0 0 90>}]
 dyn_hat_01 [{shapeType=Capsule, length=5.3102, radius=3.2983, center=<-1.4186 -0.0002 0>, rotation=<rot 90 0 0>}]
 dyn_hat_02 [{shapeType=Capsule, length=6.9317, radius=3.0948, center=<1.5973 -0.0002 0>, rotation=<rot 90 0 0>}]
 dyn_hat_03 [{shapeType=Sphere, radius=2.8601, center=<2.678 -0.0002 0>}]
 dyn_mustache_l_01 [{shapeType=Capsule, length=2, radius=2, center=<-0.7097 0 0>, rotation=<rot 90 0 0>}]
 dyn_mustache_l_02 [{shapeType=Capsule, length=2, radius=2, center=<0.1 0 0>, rotation=<rot 90 0 0>}]
 dyn_mustache_r_01 [{shapeType=Capsule, length=2, radius=2, center=<0.7097 0 0>, rotation=<rot 90 0 0>}]
 dyn_mustache_r_02 [{shapeType=Capsule, length=2, radius=2, center=<-0.1 0 0>, rotation=<rot 90 0 0>}]
 hand_r [{shapeType=Capsule, length=13.0923, radius=4.8036, center=<-7.5392 -0.0001 4.9809>, rotation=<rot 90 0 0>}]
 head [{shapeType=Sphere, radius=12.3909, center=<3.9529 2.5349 0.0001>}]
 spine_03 [{shapeType=Capsule, length=13.8288, radius=10.8392, center=<9.0364 4.1674 0>, rotation=<rot 0 -90 0>}]
 spine_04 [{shapeType=Capsule, length=13.8288, radius=10.8392, center=<6.8866 4.6095 -0.0002>, rotation=<rot 0 -90 0>}]
 spine_05 [
  {shapeType=Capsule, length=11.8744, radius=12.7935, center=<-0.6606 1.2222 -0.0001>, rotation=<rot 0 -90 0>}
  {shapeType=Capsule, length=5.79, radius=11.5001, center=<13.4021 -1.337 -0.0003>}
 ]
}
constraints [
 {bone1Name=dyn_hat_03, bone2Name=dyn_hat_02, swing1LimitDegrees=20, swing2LimitDegrees=20, twistLimitDegrees=5}
 {bone1Name=dyn_hat_02, bone2Name=dyn_hat_01, swing1LimitDegrees=75, swing2LimitDegrees=75, twistLimitDegrees=8}
 {bone1Name=dyn_mustache_l_02, bone2Name=dyn_mustache_l_01, swing1LimitDegrees=15, swing2LimitDegrees=15, twistLimitDegrees=2}
 {bone1Name=dyn_mustache_r_02, bone2Name=dyn_mustache_r_01, swing1LimitDegrees=15, swing2LimitDegrees=15, twistLimitDegrees=2}
 {
  bone1Name=C_jaw
  bone2Name=head
  swing1LimitDegrees=45
  swing1Motion=Free
  swing2LimitDegrees=45
  swing2Motion=Free
  twistLimitDegrees=45
  twistMotion=Free
 }
 {bone1Name=dyn_mustache_r_01, bone2Name=head, swing1LimitDegrees=15, swing2LimitDegrees=15, twistLimitDegrees=2}
 {bone1Name=dyn_mustache_l_01, bone2Name=head, swing1LimitDegrees=15, swing2LimitDegrees=15, twistLimitDegrees=2}
 {bone1Name=dyn_hat_01, bone2Name=head, swing1LimitDegrees=60, swing2LimitDegrees=90, twistLimitDegrees=5}
 {
  bone1Name=head
  bone2Name=spine_05
  swing1LimitDegrees=45
  swing1Motion=Free
  swing2LimitDegrees=45
  swing2Motion=Free
  twistLimitDegrees=45
  twistMotion=Free
 }
 {
  bone1Name=spine_05
  bone2Name=spine_04
  swing1LimitDegrees=45
  swing1Motion=Free
  swing2LimitDegrees=45
  swing2Motion=Free
  twistLimitDegrees=45
  twistMotion=Free
 }
 {
  bone1Name=spine_04
  bone2Name=spine_03
  swing1LimitDegrees=45
  swing1Motion=Free
  swing2LimitDegrees=45
  swing2Motion=Free
  twistLimitDegrees=45
  twistMotion=Free
 }
 {bone1Name=dyn_beard_01, bone2Name=C_jaw, swing1LimitDegrees=45, swing2LimitDegrees=45, twistLimitDegrees=10}
 {bone1Name=dyn_beard_02, bone2Name=dyn_beard_01, swing1LimitDegrees=45, swing2LimitDegrees=45, twistLimitDegrees=10}
 {bone1Name=dyn_beard_03, bone2Name=dyn_beard_02, swing1LimitDegrees=45, swing2LimitDegrees=45, twistLimitDegrees=10}
 {
  bone1Name=hand_r
  bone2Name=spine_05
  swing1LimitDegrees=45
  swing1Motion=Free
  swing2LimitDegrees=45
  swing2Motion=Free
  twistLimitDegrees=45
  twistMotion=Free
 }
]
